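# Material function for the inflatable celebration blimp that adds swimming motion and surface detail to incoming material attributes, gated by a UseInflatable switch.
# A masked sine world position offset, with timing and speed varied by per-actor custom primitive data, animates the mesh.
# UsePropeller selects either spinning fan textures via rotated UVs or panning world-aligned fold normals blended over the base normal, flattened by normal direction.
# Several experimental branches remain unconnected.
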MaterialFunction /Game/Environments/Artemis/Props/Celebration/Celebration_InflatableBlimp_A/Materials/MF_Inflatable
graph {nodes [
  n1 = ScalarParameter() {bUseCustomPrimitiveData=true, parameterName=RandomPerActor}
  n2 = Add(
   A=Time()
   B=LinearInterpolate(A=ScalarParameter() {parameterName=Min_offset}, B=ScalarParameter() {defaultValue=12, parameterName=Max_offset}, Alpha=@n1)
  ) {constB=0.45}
  Add(A=@n2)
  n3 = GetMaterialAttributes(None=FunctionInput() {inputName=Material, inputType=FunctionInput_MaterialAttributes}) {attributeGetTypes [#1, #2]}
  n4 = ScalarParameter() {defaultValue=-0.5, parameterName=IF_SwimSpeedMin}
  n5 = ScalarParameter() {defaultValue=-1, parameterName=IF_SwimSpeedMAx}
  n6 = Add(A=NamedRerouteUsage(), B=Multiply(A=@n2, B=LinearInterpolate(A=@n4, B=@n5, Alpha=@n1)))
  n7 = ScalarParameter() {defaultValue=0.25, parameterName=SineWaveFrequency}
  n8 = Multiply(A=Sine(None=Multiply(A=@n6, B=@n7)), B=VectorParameter() {parameterName=WPO_Strength, defaultValue=<rgba 0 32 0 1>}.RGB)
  n9 = Transform(None=VectorParameter() {parameterName=P1_RotationVector, defaultValue=<rgba 0 0 1 1>}.RGB) {transformSourceType=TRANSFORMSOURCE_Local}
  n10 = Multiply(A=Time(), B=ScalarParameter() {defaultValue=0.25, parameterName=P1_RotationSpeed})
  n11 = MaterialFunctionCall() {materialFunction=&ObjectPivotPoint}
  n12 = WorldPosition() {worldPositionShaderOffset=WPT_ExcludeAllShaderOffsets}
  Add(
   A=@n3.WorldPositionOffset
   B=Add(A=@n8, B=RotateAboutAxis(NormalizedRotationAxis=@n9, RotationAngle=@n10, PivotPoint=@n11.Object Pivot Location, Position=@n12.XYZ))
  )
  n13 = Subtract(A=WorldPosition().XYZ, B=ObjectPositionWS())
  Add(A=Transform(None=Normalize(VectorInput=@n13)) {transformSourceType=TRANSFORMSOURCE_Local}, B=@n8)
  n14 = VertexNormalWS()
  n15 = Abs(None=DotProduct(A=Constant3Vector() {constant=<rgba 0 0 1 0>}, B=@n14))
  n16 = Abs(None=DotProduct(A=Constant3Vector() {constant=<rgba 0 1 0 0>}, B=@n14))
  Add(A=@n15, B=@n16)
  n17 = Time()
  n18 = ScalarParameter() {bUseCustomPrimitiveData=true, parameterName=RandomPerActor}
  n19 = LinearInterpolate(A=ScalarParameter() {parameterName=Min_offset}, B=ScalarParameter() {defaultValue=12, parameterName=Max_offset}, Alpha=@n18)
  n20 = ScalarParameter() {defaultValue=-0.5, parameterName=IF_SwimSpeedMin}
  n21 = ScalarParameter() {defaultValue=-1, parameterName=IF_SwimSpeedMAx}
  n22 = Add(A=NamedRerouteUsage(), B=Multiply(A=Add(A=@n17, B=@n19), B=LinearInterpolate(A=@n20, B=@n21, Alpha=@n18)))
  n23 = ScalarParameter() {defaultValue=0.25, parameterName=SineWaveFrequency}
  n24 = Multiply(A=Sine(None=Multiply(A=@n22, B=@n23)), B=VectorParameter() {parameterName=WPO_Strength, defaultValue=<rgba 0 32 0 1>}.RGB)
  n25 = Add(
   A=@n3.WorldPositionOffset
   B=LinearInterpolate(A=Transform(None=@n24) {transformSourceType=TRANSFORMSOURCE_Local}, Alpha=NamedRerouteUsage()) {constB=0}
  )
  n26 = Rotator(Time=Multiply(A=Time(), B=ScalarParameter() {defaultValue=25, parameterName=PropellerSpeed}))
  n27 = TextureSample(UVs=@n26) {
   samplerSource=SSM_Wrap_WorldGroupSettings
   samplerType=SAMPLERTYPE_LinearColor
   texture=/Game/Environments/Artemis/Props/Celebration/Celebration_InflatableBlimp_A/Textures/T_Celebration_InflatableBlimpFan_A_A_D
  }
  n28 = TextureSample(UVs=@n26) {
   samplerSource=SSM_Wrap_WorldGroupSettings
   samplerType=SAMPLERTYPE_Masks
   texture=/Game/Environments/Artemis/Props/Celebration/Celebration_InflatableBlimp_A/Textures/T_Celebration_InflatableBlimpFan_A_A_S
  }
  n29 = TextureSample(UVs=@n26) {
   samplerSource=SSM_Wrap_WorldGroupSettings
   samplerType=SAMPLERTYPE_Normal
   texture=/Game/Environments/Artemis/Props/Celebration/Celebration_InflatableBlimp_A/Textures/T_Celebration_InflatableBlimpFan_A_A_N
  }
  n30 = SetMaterialAttributes!(
   MaterialAttributes=@n3.MaterialAttributes
   WorldPositionOffset=@n25
   EmissiveColor
   BaseColor=@n27.RGB
   OpacityMask=@n27.RGB
   Metallic=@n28.G
   Specular=@n28.G
   Normal=@n29.RGB
   Roughness=@n28.G
  ) {attributeSetTypes [
    #1
    #3
    #4
    #5
    #6
    #7
    #2
    #8
   ]}
  n31 = Constant3Vector() {constant=<rgba 0 0 1 0>}
  n32 = ScalarParameter() {defaultValue=0.872, parameterName=Rotate}
  n33 = Constant3Vector()
  n34 = NamedRerouteUsage()
  n35 = Time()
  n36 = VectorParameter() {parameterName=FoldsPanSpeedAndDirection, defaultValue=<rgba 0 128 0 1>}
  n37 = MaterialFunctionCall(
   TextureObject=TextureObject() {samplerType=SAMPLERTYPE_Normal, texture=/Game/Gadgets/Assets/VinderTech_GliderChute/Glider_BlackMondayCape/Textures/T_Folds_N}
   TextureSize=VectorParameter() {parameterName=Folds_TilingSize, defaultValue=<rgba 256 128 128 1>}.RGB
   Normal=NamedRerouteUsage()
   WorldPosition=Add(A=RotateAboutAxis(NormalizedRotationAxis=@n31, RotationAngle=@n32, PivotPoint=@n33, Position=@n34), B=Multiply(A=@n35, B=@n36.RGB))
  ) {materialFunction=&WorldAlignedNormal}
  n38 = DotProduct(A=Constant3Vector() {constant=<rgba 1 0 0 0>}, B=@n14)
  n39 = OneMinus(None=Multiply(A=Multiply(A=@n15, B=@n16), B=Abs(None=@n38)))
  n40 = Multiply(A=Subtract(A=@n39) {constB=0.9}) {constB=3}
  n41 = LinearInterpolate(
   A=ScalarParameter() {defaultValue=0.85, parameterName=NormalFlatnessMin}
   B=ScalarParameter() {parameterName=FoldsFlatness}
   Alpha=Clamp(None=Power(Base=@n40))
  ) {constA=1}
  n42 = SetMaterialAttributes!(
   MaterialAttributes=@n3.MaterialAttributes
   WorldPositionOffset=@n25
   Normal=MaterialFunctionCall(
    BaseNormal=@n3.Normal
    AdditionalNormal=MaterialFunctionCall(Normal=@n37.XYZ Texture, Flatness=@n41) {materialFunction=&FlattenNormal}.Result
   ) {materialFunction=&BlendAngleCorrectedNormals}.Result
  ) {attributeSetTypes [#1, #2]}
  FunctionOutput(None=StaticSwitchParameter(True=StaticSwitchParameter(True=@n30, False=@n42) {parameterName=UsePropeller}, False=@n3.MaterialAttributes) {parameterName=UseInflatable})
  n43 = TextureSample(UVs=BumpOffset(
    Coordinate=@n26
    Height=ScalarParameter() {parameterName=Height}
    HeightRatioInput=ScalarParameter() {defaultValue=0.05, parameterName=HeightRatio}
   ) {heightRatio=1}) {
   samplerSource=SSM_Wrap_WorldGroupSettings
   samplerType=SAMPLERTYPE_LinearColor
   texture=&T_Celebration_InflatableBlimpFan_A_A_D
  }
  Max(A=@n27.A, B=@n43.A)
  n44 = PreSkinnedPosition()
  n45 = ShaderStageSwitch(PixelShader=VertexInterpolator(VS=@n44).PS, VertexShader=@n44)
  n46 = Add(A=@n45, B=VectorParameter() {parameterName=PivotOffset, defaultValue=<rgba -380.7562 0 0 1>}.RGB)
  n47 = PreSkinnedLocalBounds()
  NamedRerouteDeclaration(None=ChannelMaskParameter(None=Divide(A=@n46, B=@n47.Max)) {parameterName=FwdGradient}.R) {name=GradientAlongMesh, nodeColor=<rgba 1 0 0.7031 1>}
  NamedRerouteDeclaration(None=@n45) {name=Preskinnedposition, nodeColor=<rgba 0 1 0.9062 1>}
  n48 = PreSkinnedNormal()
  NamedRerouteDeclaration(None=ShaderStageSwitch(PixelShader=VertexInterpolator(VS=@n48).PS, VertexShader=@n48)) {name=PreskinnedNormal, nodeColor=<rgba 0 0.6719 1 1>}
  n49 = PreSkinnedPosition()
  n50 = Add(
   A=ShaderStageSwitch(PixelShader=VertexInterpolator(VS=@n49).PS, VertexShader=@n49)
   B=Transform(None=VectorParameter() {parameterName=P_PivotOffset, defaultValue=<rgba -208 -196 0 1>}.RGB) {transformSourceType=TRANSFORMSOURCE_Local}
  )
  n51 = Transform(None=VectorParameter() {parameterName=BoundsValue, defaultValue=<rgba 380 248 194 1>}.RGB) {transformSourceType=TRANSFORMSOURCE_Local}
  NamedRerouteDeclaration(None=ChannelMaskParameter(None=Divide(A=@n50, B=@n51)) {parameterName=FwdGradient}.R) {name=P_GradientAlongMesh, nodeColor=<rgba 1 0 0.7031 1>}
  NamedRerouteDeclaration(None=SphereMask(
    A=NamedRerouteUsage()
    B=VectorParameter() {parameterName=WPO_MaskPos, defaultValue=<rgba -72 -36 -196 1>}.RGB
    Radius=ScalarParameter() {defaultValue=200, parameterName=MaskRadius}
    Hardness=ScalarParameter() {defaultValue=0.25, parameterName=MaskHardness}
   )) {name=WPOMaskPos, nodeColor=<rgba 0.2188 0 1 1>}
  PerInstanceRandom()
  Power(Base=@n39) {constExponent=12}
  PreSkinnedLocalBounds()
  ScalarParameter() {parameterName=RandomPerActor}
  ScalarParameter() {parameterName=RandomPerActor}
  ScalarParameter() {defaultValue=-1, parameterName=IF_SwimSpeed}
  ScalarParameter() {defaultValue=0.81, parameterName=TimeOffset}
  Transform(None=@n8) {transformSourceType=TRANSFORMSOURCE_Local}
  VectorParameter() {parameterName=P1_Location, defaultValue=<rgba 0 0 0 1>}
  WorldPosition()
 ]}
