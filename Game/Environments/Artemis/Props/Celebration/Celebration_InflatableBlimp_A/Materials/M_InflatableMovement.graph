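# Material for the inflatable celebration blimp, building swaying motion and moving fabric folds.
# A sine wave, varied per actor through custom primitive data, forms a directional offset, alongside two time-driven pivot rotations selected by red vertex-color masks.
# A panning world-aligned fold normal is flattened by a mask from normal direction.
# Pre-skinned position and bounds yield a gradient along the mesh; no connections to material outputs appear.
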Material /Game/Environments/Artemis/Props/Celebration/Celebration_InflatableBlimp_A/Materials/M_InflatableMovement
graph {
 properties {pixelDepthOffsetMode=PDOM_Legacy, usedShadingModels=Default Lit}
 nodes [
  n1 = WorldPosition()
  n2 = ObjectPositionWS()
  n3 = Time()
  n4 = ScalarParameter() {bUseCustomPrimitiveData=true, parameterName=RandomPerActor}
  n5 = LinearInterpolate(A=ScalarParameter() {parameterName=Min_offset}, B=ScalarParameter() {defaultValue=12, parameterName=Max_offset}, Alpha=@n4)
  n6 = ScalarParameter() {defaultValue=-0.5, parameterName=IF_SwimSpeedMin}
  n7 = ScalarParameter() {defaultValue=-1, parameterName=IF_SwimSpeedMAx}
  n8 = Add(A=NamedRerouteUsage(), B=Multiply(A=Add(A=@n3, B=@n5), B=LinearInterpolate(A=@n6, B=@n7, Alpha=@n4)))
  n9 = ScalarParameter() {defaultValue=0.25, parameterName=SineWaveFrequency}
  n10 = Multiply(A=Sine(None=Multiply(A=@n8, B=@n9)), B=VectorParameter() {parameterName=WPO_Strength, defaultValue=<rgba 0 64 0 1>}.RGB)
  Add(A=Normalize(VectorInput=Subtract(A=@n1.XYZ, B=@n2)), B=@n10)
  n11 = Transform(None=@n10) {transformSourceType=TRANSFORMSOURCE_World, transformType=TRANSFORM_Local}
  n12 = Time()
  n13 = ScalarParameter() {defaultValue=0.25, parameterName=P1_RotationSpeed}
  n14 = WorldPosition()
  n15 = Time()
  n16 = ScalarParameter() {defaultValue=0.25, parameterName=P2_RotationSpeed}
  n17 = VertexColor()
  n18 = Constant() {r=0.1}
  n19 = LinearInterpolate(
   A=RotateAboutAxis(
    NormalizedRotationAxis=VectorParameter() {parameterName=P1_RotationVector, defaultValue=<rgba 0 0 1 1>}.RGB
    RotationAngle=Multiply(A=@n12, B=@n13)
    PivotPoint=VectorParameter() {parameterName=P1_Location, defaultValue=<rgba 0 0 0 1>}.RGB
    Position=@n14.XYZ
   )
   B=RotateAboutAxis(
    NormalizedRotationAxis=VectorParameter() {parameterName=P2_RotationVector, defaultValue=<rgba 0 0 1 1>}.RGB
    RotationAngle=Multiply(A=@n15, B=@n16)
    PivotPoint=VectorParameter() {parameterName=P2_Location, defaultValue=<rgba 0 0 0 1>}.RGB
    Position=@n14.XYZ
   )
   Alpha=SphereMask(A=@n17.R, B=ScalarParameter() {defaultValue=0.5, parameterName=P2_Sample}, Radius=@n18)
  )
  Add(A=@n11, B=@n19)
  n20 = VertexNormalWS()
  n21 = Abs(None=DotProduct(A=Constant3Vector() {constant=<rgba 0 0 1 0>}, B=@n20))
  n22 = Abs(None=DotProduct(A=Constant3Vector() {constant=<rgba 0 1 0 0>}, B=@n20))
  Add(A=@n21, B=@n22)
  Constant3Vector() {constant=<rgba 0 0 1 0>}
  LinearInterpolate(B=@n19, Alpha=SphereMask(A=@n17.R, B=ScalarParameter() {defaultValue=0.25, parameterName=P1_Sample}, Radius=@n18))
  n23 = RotateAboutAxis(
   NormalizedRotationAxis=Constant3Vector() {constant=<rgba 0 0 1 0>}
   RotationAngle=ScalarParameter() {defaultValue=0.872, parameterName=Rotate}
   PivotPoint=Constant3Vector()
   Position=NamedRerouteUsage()
  )
  n24 = Multiply(A=Time(), B=VectorParameter() {parameterName=FoldsPanSpeedAndDirection, defaultValue=<rgba 0 128 0 1>}.RGB)
  n25 = DotProduct(A=Constant3Vector() {constant=<rgba 1 0 0 0>}, B=@n20)
  n26 = OneMinus(None=Multiply(A=Multiply(A=@n21, B=@n22), B=Abs(None=@n25)))
  n27 = Power(Base=Multiply(A=Subtract(A=@n26) {constB=0.9}) {constB=3})
  MaterialFunctionCall(
   Normal=MaterialFunctionCall(
    TextureObject=TextureObject() {samplerType=SAMPLERTYPE_Normal, texture=/Game/Gadgets/Assets/VinderTech_GliderChute/Glider_BlackMondayCape/Textures/T_Folds_N}
    TextureSize=VectorParameter() {parameterName=Folds_TilingSize, defaultValue=<rgba 256 128 128 1>}.RGB
    Normal=NamedRerouteUsage()
    WorldPosition=Add(A=@n23, B=@n24)
   ) {materialFunction=&WorldAlignedNormal}.XYZ Texture
   Flatness=LinearInterpolate(B=ScalarParameter() {parameterName=FoldsFlatness}, Alpha=Clamp(None=@n27)) {constA=1}
  ) {materialFunction=&FlattenNormal}
  n28 = PreSkinnedLocalBounds()
  MaterialFunctionCall!(Vector3=@n28.Max, MaximumNumberOfDigits, UVs, DebugTextLocation RG_UpperRight BA_LowerLeft, Component Spacing) {materialFunction=/Engine/Functions/Engine_MaterialFunctions02/Utility/DebugFloat3Values}
  n29 = PreSkinnedPosition()
  n30 = ShaderStageSwitch(PixelShader=VertexInterpolator(VS=@n29).PS, VertexShader=@n29)
  n31 = Add(A=@n30, B=VectorParameter() {parameterName=PivotOffset, defaultValue=<rgba -380.7562 0 0 1>}.RGB)
  NamedRerouteDeclaration(None=ChannelMaskParameter(None=Divide(A=@n31, B=@n28.Max)) {parameterName=FwdGradient}.R) {name=GradientAlongMesh, nodeColor=<rgba 1 0 0.7031 1>}
  NamedRerouteDeclaration(None=@n30) {name=Preskinnedposition, nodeColor=<rgba 0 1 0.9063 1>}
  n32 = PreSkinnedNormal()
  NamedRerouteDeclaration(None=ShaderStageSwitch(PixelShader=VertexInterpolator(VS=@n32).PS, VertexShader=@n32)) {name=PreskinnedNormal, nodeColor=<rgba 0 0.6719 1 1>}
  NamedRerouteUsage()
  PerInstanceRandom()
  Power(Base=@n26) {constExponent=12}
  Reroute(None=@n11)
  ScalarParameter() {parameterName=RandomPerActor}
  ScalarParameter() {bUseCustomPrimitiveData=true, parameterName=RandomPerActor}
  VertexColor()
  WorldPosition()
 ]
}
